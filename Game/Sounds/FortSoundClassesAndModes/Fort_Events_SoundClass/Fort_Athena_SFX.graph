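# Sound class for Athena sound effects, a child of the in-game SFX class.
# It groups two child classes, Athena player foley and consumables.
# Ambient volumes apply to it, and reverb is turned off.
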SoundClass /Game/Sounds/FortSoundClassesAndModes/Fort_Events_SoundClass/Fort_Athena_SFX
parent=&Fort_SFX_InGame
children [&Fort_Player_Foley_Athena, &Fort_Consumables]
properties {properties {bApplyAmbientVolumes=true, bReverb=false}}
